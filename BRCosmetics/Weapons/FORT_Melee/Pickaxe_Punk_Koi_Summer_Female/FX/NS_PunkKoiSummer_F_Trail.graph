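# Niagara swing-trail effect for the female Punk Koi Summer pickaxe, layering a ribbon, glint and sheen.
# AirTrail builds a short-lived anim-trail ribbon from bone or socket positions, blending User.TrailColor toward a brighter hue-shifted version via User.Glow.
# DualColorGlint001 bursts one particle into two sprite renderers, with alternate position, size, color and material attributes animated separately.
# SheenMesh_start bursts one camera-offset mesh with an animated material parameter and fading alpha.
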NiagaraSystem /BRCosmetics/Weapons/FORT_Melee/Pickaxe_Punk_Koi_Summer_Female/FX/NS_PunkKoiSummer_F_Trail
dependencies {
 usedDataInterfaces [&NiagaraDataInterfaceCurve, &NiagaraDataInterfaceVector2DCurve]
 usedDynamicInputs []
 usedModules [&SystemState]
}
emitters {
 AirTrail {
  emitterUpdateScript [
   &EmitterState(Life Cycle Mode=::NewEnumerator1, Loop Behavior=::NewEnumerator1, Loop Duration=0.3, UseLoopDelay=false)
   &FN_SpawnAnimTrail_v2(Spawn Interval=0)
  ]
  particleSpawnScript [
   /Game/Effects/Niagara/Modules/AnimTrail/FN_InitializeAnimTrail/FN_InitializeAnimTrailBoneIndecies(First Bone or Socket Index=1)
   /Game/Effects/Niagara/Modules/AnimTrail/FN_InitializeAnimTrail/FN_InitializeAnimTrail(
    AnimTrailWidth=1.25
    Color=&MultiplyLinearColor(
     Linear Color=&LerpLinearColors(
      StartColor=$User.TrailColor
      EndColor=&HueShiftLinearColor(Hue=-0.05, Saturation=5, Value=10, Linear Color=$User.TrailColor)
      LerpFactor=$User.Glow
     )
     Scale Factor=&MakeFloat4FromVectorAndFloat(
      "Vector (RGB)"=<1 1 1>
      "Float (Alpha)"=&FloatFromCurve(FloatCurve=DataInterface() {propertyValues {Curve {keys [<at 0.07 1>,<at 0.125 0>]}}}, CurveIndex=$Emitter.Age)
     )
    )
    Lifetime=0.125
   )
  ]
  particleUpdateScript [
   &SolveForcesAndVelocity(Clamp Velocity=false, Limit Acceleration=false)
   &FN_UpdateAnimTrail_v2()
   &ScaleColor(
    Scale Alpha=&FloatFromCurve(FloatCurve=DataInterface() {propertyValues {Curve {keys [<at 0 1 RCIM_Cubic>,<at 1 0>]}}})
    ScaleA=true
    ScaleRGB=true
   )
  ]
  rendererClasses [&NiagaraRibbonRendererProperties]
  renderers [{rendererClass=&NiagaraRibbonRendererProperties}]
 }
 DualColorGlint001 {
  emitterUpdateScript [
   &EmitterState(Life Cycle Mode=::NewEnumerator1, Loop Behavior=::NewEnumerator1, Loop Delay=0.2, Loop Duration=1, UseLoopDelay=false)
   &SpawnBurst_Instantaneous(Spawn Count=1, Use Spawn Probability=false)
  ]
  particleSpawnScript [
   &InitializeParticle(
    Color=&MakeLinearColorFromVectorAndFloat("Vector (RGB)"=<1 1 1>, "Float (Alpha)"=&RandomRangeFloat(Minimum=0.2, Maximum=1))
    Lifetime=0.3
    Mesh Scale Mode=::NewEnumerator3
    Position Mode=::NewEnumerator1
    Sprite Size=<20 10>
    Sprite Size Mode=::NewEnumerator4
    UsePositionOffset=false
    Write Material Random=true
   )
   &AddVelocity(Velocity=<0 0 5>)
  ]
  particleUpdateScript [
   &ParticleState()
   &DynamicMaterialParameters(
    Index 0 Param 1=1
    Index 0 Param 2=&FloatFromCurve(FloatCurve=DataInterface() {propertyValues {Curve {keys [{tangentMode=RCTM_User, time=0, value=0.1}, {tangentMode=RCTM_User, time=1, value=2}]}}})
    Index 0 Param 4=1
    Module.Param0WriteEnabled=true
    Module.Param1WriteEnabled=true
    Module.Param2WriteEnabled=true
    Module.Param3WriteEnabled=true
   )
   /SetVariables(Particles.DynaMatAlt:Vector4f=&MakeVector4(
     X=1
     Y=&FloatFromCurve(FloatCurve=DataInterface() {propertyValues {Curve {keys [{tangentMode=RCTM_User, time=0.3, value=1.5}, {tangentMode=RCTM_User, time=0.8, value=3}]}}})
     Z=0
     W=0
    )) {isSetParameters=true}
   &ScaleColor(
    Scale Alpha=&FloatFromCurve(
     FloatCurve=DataInterface() {propertyValues {Curve {keys [
         {arriveTangent=4.3823, interpMode=RCIM_Cubic, leaveTangent=4.3823, tangentMode=RCTM_User, time=0.2, value=0}
         {arriveTangent=-2.5249, interpMode=RCIM_Cubic, leaveTangent=-2.5249, tangentMode=RCTM_User, time=0.5, value=0.9847}
         {arriveTangent=-0.0405, interpMode=RCIM_Cubic, leaveTangent=-0.0405, tangentMode=RCTM_User, time=1, value=0}
        ]}}}
     Scale Curve=0.05
    )
    Scale RGB=&VectorFromFloat(Value=5)
    ScaleA=true
    ScaleRGB=true
   )
   /SetVariables(Particles.ColorAlt:LinearColor=&LerpLinearColors(
     StartColor=<rgba 0.1234 1 0.2918 1>
     EndColor=&LerpLinearColors(
      StartColor=<rgba 0.2727 1 0.7753 1>
      EndColor=<rgba 1 0 0.04 1>
      LerpFactor=&RandomRangeFloat(Minimum=0, Maximum=1)
     )
     LerpFactor=$Particles.NormalizedAge
    )) {isSetParameters=true}
   /SetVariables(
    Particles.PositionSprite:NiagaraPosition=&MakeVector(
     X=0
     Y=0
     Z=&Add_Float(
      A=&FloatFromCurve(
       FloatCurve=DataInterface() {propertyValues {Curve {keys [
           {arriveTangent=0.0136, interpMode=RCIM_Cubic, time=0, value=0}
           {arriveTangent=1.8987, interpMode=RCIM_Cubic, leaveTangent=1.8987, tangentMode=RCTM_User, time=0.5, value=0.45}
           {interpMode=RCIM_Cubic, leaveTangent=1.0398, time=1, value=1}
          ]}}}
       Scale Curve=60
      )
      B=25
     )
    )
    Particles.PositionSpriteAlt:NiagaraPosition=&MakeVector(
     X=0
     Y=0
     Z=&Add_Float(
      A=&FloatFromCurve(
       FloatCurve=DataInterface() {propertyValues {Curve {keys [
           {arriveTangent=0.0136, interpMode=RCIM_Cubic, time=0, value=0}
           {arriveTangent=1.8987, interpMode=RCIM_Cubic, leaveTangent=1.8987, tangentMode=RCTM_User, time=0.5, value=0.45}
           {interpMode=RCIM_Cubic, leaveTangent=1.0398, time=1, value=1}
          ]}}}
       Scale Curve=60
      )
      B=30
     )
    )
   ) {isSetParameters=true}
   &ScaleSpriteSize(
    Non-Uniform Curve Scale=<4 4>
    Non-Uniform Curve Sprite Scale=DataInterface() {propertyValues {
      XCurve {keys [
        {interpMode=RCIM_Cubic, tangentMode=RCTM_User, time=0, value=0}
        {interpMode=RCIM_Cubic, tangentMode=RCTM_User, time=0.5, value=1}
        {tangentMode=RCTM_User, time=1, value=0}
       ]}
      YCurve {keys [
        {interpMode=RCIM_Cubic, tangentMode=RCTM_User, time=0, value=0}
        {interpMode=RCIM_Cubic, tangentMode=RCTM_User, time=0.5, value=1}
        {tangentMode=RCTM_User, time=1, value=0}
       ]}
     }}
    Scale Sprite Size Mode=::NewEnumerator3
   )
   /SetVariables(Particles.SpriteSizeAlt:Vector2f=&Vector2DFromCurve(
     Vector2Curve=DataInterface() {propertyValues {
       XCurve {keys [
         {interpMode=RCIM_Cubic, tangentMode=RCTM_User, time=0, value=0}
         {interpMode=RCIM_Cubic, tangentMode=RCTM_User, time=0.5, value=1}
         {tangentMode=RCTM_User, time=1, value=0}
        ]}
       YCurve {keys [{interpMode=RCIM_Cubic, tangentMode=RCTM_User, time=0, value=1}, {interpMode=RCIM_Cubic, tangentMode=RCTM_User, time=1, value=0}]}
      }}
     ScaleCurve=<40 200>
    )) {isSetParameters=true}
   &SolveForcesAndVelocity(Clamp Velocity=false, Limit Acceleration=false)
   &SpriteRotationRate(Rotation Rate=3)
  ]
  rendererClasses [&NiagaraSpriteRendererProperties]
  renderers [
   {rendererClass=&NiagaraSpriteRendererProperties}
   {rendererClass=&NiagaraSpriteRendererProperties, rendererIndex=1}
  ]
 }
 SheenMesh_start {
  emitterUpdateScript [
   &EmitterState(Life Cycle Mode=::NewEnumerator1, Loop Behavior=::NewEnumerator1, Loop Duration=0.5, UseLoopDelay=false)
   &SpawnBurst_Instantaneous(Spawn Count=1, Use Spawn Probability=false)
  ]
  particleSpawnScript [&InitializeParticle(
    Color=<rgba 1.6 2 1.6667 1>
    Lifetime=0.25
    Mesh Scale Mode=::NewEnumerator3
    Position Offset=<-0.5 0 0.5>
    UsePositionOffset=true
    Write Material Random=true
   ) {module=InitializeParticle001}]
  particleUpdateScript [
   &ParticleState()
   &DynamicMaterialParameters(
    Index 0 Param 1=&FloatFromCurve(FloatCurve=DataInterface() {propertyValues {Curve {keys [{arriveTangent=2.513, interpMode=RCIM_Cubic, time=0, value=0.2}, <at 1 1.6>]}}})
    Index 0 Param 2=1
    Index 0 Param 3=1
    Index 0 Param 4=1
    Module.Param0WriteEnabled=true
    Module.Param1WriteEnabled=true
    Module.Param2WriteEnabled=true
    Module.Param3WriteEnabled=true
   )
   &CameraOffset(Camera Offset Amount=1)
   &ScaleColor(
    Scale Alpha=&FloatFromCurve(FloatCurve=DataInterface() {propertyValues {Curve {keys [<at 0.5 1>,<at 1 0>]}}})
    ScaleA=true
    ScaleRGB=true
   )
  ]
  rendererClasses [&NiagaraMeshRendererProperties]
  renderers [{rendererClass=&NiagaraMeshRendererProperties, rendererIndex=0}]
 }
}
summary {
 systemName=NS_PunkKoiSummer_F_Trail
 emitters [
  {emitterName=AirTrail, rendererClasses [&NiagaraRibbonRendererProperties]}
  {emitterName=DualColorGlint001, rendererClasses [&NiagaraSpriteRendererProperties]}
  {emitterName=SheenMesh_start, rendererClasses [&NiagaraMeshRendererProperties]}
 ]
 userVariables [
  User.Glow {type=NiagaraFloat, defaultValue=0, description=""}
  User.RibbonIDOffset {type=NiagaraInt32}
  User.SkeletalMesh {
   type=NiagaraDataInterfaceSkeletalMesh
   defaultValue {
    struct=NiagaraExt_VariableValue_DataInterface
    value {dataInterface=&self:NiagaraDataInterfaceSkeletalMesh_3}
   }
  }
  User.TrailColor {type=LinearColor, description="", defaultValue {struct=LinearColor, value=<rgba 0.2893 0.5 0.3596 1>}}
 ]
}
